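; Parent gameplay ability for Season 27 Bootcamp quest granters, which gives the player an urgent tutorial quest in Battle Royale playlists.
; On activation it checks the game phase, waiting for the bus skydive to end if in Aircraft, then delays briefly before granting the transient quest that children define.
; When the matching quest is completed or claimed, it broadcasts a custom verb to complete the granter and ends, unbinding its listener.
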(event Ability|EventActivateAbility
  ;[0 0 0] Activate Ability - Check if player is in BR, and check the current game phase to determine when boot camp should grant
  (|CallParentFunction)
  (bind _bbr (CallFunction|PlaylistIsBattleRoyale))
  (if _bbr
    (Variables|Default|SetOwningPlayerPawn (CallFunction|GetAvatarasFortPlayerPawn))
    (CallFunction|CheckAthenaGamePhase)
    (else
      (Ability|EndAbility))))

(event Custom|CheckAthenaGamePhase
  (bind _returnvalue (BattleRoyaleGamePhaseLogic|GetBattleRoyaleGamePhaseLogicComponent))
  ;[0 0 0] Check against the current BR game phase - if on the bus, wait until landed from bus; if in gameplay phase, start delay for the tutorial urgent quest grant
  (Utilities|IsValid _returnvalue
    (:Is%20Valid
      (switch Utilities|FlowControl|Switch|SwitchonEAthenaGamePhase (BattleRoyaleGamePhaseLogic|GetGamePhase _returnvalue)
        (:Aircraft
          (Ability|Tasks|WaitGameplayEvent "(TagName=\"Event.SkydiveEnded.FromBus\")" 0 true
            (:EventReceived
              (Development|PrintString "Landed from bus, starting delay for tutorial..." true true "(R=1.000000,G=0.000000,B=0.222520,A=1.000000)" 3.0))))
        (:SafeZones
          (Development|PrintString "Already in gameplay phase, starting delay for tutorial..." true true "(R=1.000000,G=0.000000,B=0.222520,A=1.000000)" 3.0)))
      (Utilities|FlowControl|Delay 5.0)
      (CallFunction|GrantUrgentQuest))
    (:Is%20Not%20Valid
      (Ability|EndAbility))))

(event Custom|GrantUrgentQuest
  (bind _owningplayerpawn (Variables|Default|GetOwningPlayerPawn))
  (bind _returnvalue (Class|FortPlayerPawn|GetTransientQuestsComponent _owningplayerpawn))
  ;[0 0 0] Grant the urgent quest as defined in class defaults on children and then bind event to when the urgent quest ends
  (Class|FortControllerComponentTransientQuests|GrantTransientQuest _returnvalue (Variables|UrgentQuestVariables|GetUrgent%20Quest%20to%20Grant) 0)
  (EventDispatchers|BindEventtoOnUrgentQuestEnded _returnvalue (EventDispatchers|CreateEvent :!event Custom|On%20Urgent%20Quest%20Ended)))

(event Custom|On%20Urgent%20Quest%20Ended
  (params (PlayerStateAthena :type Fort%20Player%20State%20Athena%20Object%20Reference) (UrgentQuestTag :type Gameplay%20Tag%20Structure%20%28by%20ref%29) (QuestState :type EFortQuestState%20Enum :default "Inactive") (CompletionPathContext :type Gameplay%20Tag%20Container%20Structure))
  ;[0 0 0] When urgent quest ends, check if it is the current urgent quest; then send quest update for the granter to complete it
  (if (== (Variables|UrgentQuestVariables|GetUrgent%20Quest%20Tag) UrgentQuestTag)
    (if (or (Utilities|Enum|Equal(Enum) QuestState "Completed") (Utilities|Enum|Equal(Enum) QuestState "Claimed"))
      (CustomVerb|BroadcastCustomVerbMessage (CustomVerb|MakeCustomVerbMessage (Variables|Default|GetOwningPlayerPawn) (Variables|UrgentQuestVariables|GetCustomVerb)))
      (Ability|EndAbility))))

(event Ability|EventFailedToActivatePassiveAbility
  (params (ActorInfo :type Gameplay%20Ability%20Actor%20Info%20Structure))
  ;[0 0 0] On End ability remove bind; on ability failed to activate logic
  (|CallParentFunction ActorInfo))

(event Ability|EventOnEndAbility
  (params (bWasCancelled :type Boolean))
  ;[0 0 0] On End ability remove bind; on ability failed to activate logic
  (|CallParentFunction bWasCancelled)
  (EventDispatchers|UnbindEventfromOnUrgentQuestEnded (Class|FortPlayerPawn|GetTransientQuestsComponent (CallFunction|GetAvatarasFortPlayerPawn)) (EventDispatchers|CreateEvent :!event Custom|On%20Urgent%20Quest%20Ended)))

(fn PlaylistIsBattleRoyale
  (return (Class|FortGameStateAthena|IsBattleRoyalePlaylist (Fortnite|Athena|GameState|GetGameStateAthenafromContext (CallFunction|GetAvatarasFortPlayerPawn)) false)))
